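# Translucent Niagara sprite material for the PixelTech vehicle boost that draws a procedural, texture-free glowing star.
# Squared, centered UVs form a thin upright cross and a diagonally rotated copy, both masked by a radial falloff, with fixed-time Panner and Rotator nodes as static transforms.
# Dynamic parameters set each cross's exponent, their shared gain and brightness.
# Particle color tints the sum before MF_FX_EmissiveAdjustments, while particle alpha scales a separate clamped mask.
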Material /VehicleCosmetics/Boosts/Boost_PixelTech/Materials/MM_GlowBits
graph {
 properties {bAutomaticallySetUsageInEditor=false, bUsedWithNiagaraSprites=true, blendMode=BLEND_Translucent, usedShadingModels=Default Lit}
 nodes [
  n1 = Constant() {r=60}
  n2 = Add(A=TextureCoordinate(), B=Constant() {r=-0.5})
  n3 = Multiply(A=@n2, B=@n2)
  n4 = Add(A=Constant(), B=ComponentMask(None=@n3) {r=true})
  MaterialExpressionConstantClamp_37 = Saturate(None=OneMinus(None=Multiply(A=@n1, B=@n4)))
  n5 = Constant() {r=10}
  n6 = Add(A=ComponentMask(None=@n3) {g=true}, B=Constant())
  n7 = Constant() {r=60}
  MaterialExpressionConstantClamp_36 = Saturate(None=OneMinus(None=Multiply(A=@n6, B=@n7)))
  n8 = Add(A=Panner(Coordinate=TextureCoordinate() {uTiling=3, vTiling=3}, Time=Constant() {r=1}) {speedX=1.5, speedY=1.5}, B=Constant() {r=-3})
  n9 = DotProduct(A=@n8, B=@n8)
  MaterialExpressionConstantClamp_38 = Saturate(None=ConstantBiasScale(None=OneMinus(None=@n9)))
  n10 = Multiply(A=@MaterialExpressionConstantClamp_38, B=@MaterialExpressionConstantClamp_38)
  n11 = Multiply(
   A=LinearInterpolate(
    A=Power(Base=@MaterialExpressionConstantClamp_37, Exp=@n5)
    B=Power(Base=@MaterialExpressionConstantClamp_36, Exp=@n5)
    Alpha=Constant() {r=0.5}
   )
   B=@n10
  )
  n12 = DynamicParameter() {paramNames [Brightness,Param2,Power,Param4]}
  n13 = Constant() {r=60}
  n14 = Add(A=Rotator(Coordinate=TextureCoordinate(), Time=Constant() {r=0.75}) {speed=3.14}, B=Constant() {r=-0.5})
  n15 = Multiply(A=@n14, B=@n14)
  n16 = Add(A=Constant(), B=ComponentMask(None=@n15) {r=true})
  MaterialExpressionConstantClamp_39 = Saturate(None=OneMinus(None=Multiply(A=@n13, B=@n16)))
  n17 = Constant() {r=10}
  n18 = Add(A=ComponentMask(None=@n15) {g=true}, B=Constant())
  n19 = Constant() {r=60}
  MaterialExpressionConstantClamp_40 = Saturate(None=OneMinus(None=Multiply(A=@n18, B=@n19)))
  n20 = Multiply(
   A=LinearInterpolate(
    A=Power(Base=@MaterialExpressionConstantClamp_39, Exp=@n17)
    B=Power(Base=@MaterialExpressionConstantClamp_40, Exp=@n17)
    Alpha=Constant() {r=0.5}
   )
   B=@n10
  )
  n21 = Add(A=Multiply(A=Power(Base=@n11, Exp=@n12.Power), B=@n12.Param2), B=Multiply(A=Power(Base=@n20, Exp=@n12.Param4), B=@n12.Param2))
  MaterialExpressionVertexColor_8 = ParticleColor()
  MaterialExpressionConstantClamp_35 = Saturate(None=Multiply(A=@n21, B=@MaterialExpressionVertexColor_8.A))
  MaterialFunctionCall(Emissive=Multiply(A=Multiply(A=@n21, B=@MaterialExpressionVertexColor_8.RGB), B=@n12.Brightness)) {materialFunction=&MF_FX_EmissiveAdjustments}
 ]
}
