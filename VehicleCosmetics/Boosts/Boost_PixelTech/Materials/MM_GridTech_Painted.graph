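# Unlit masked master material for the PixelTech boost effect, usable on Niagara ribbons, drawing an animated, particle-tinted pixel grid.
# Tiled, per-axis panned UVs feed a time-driven flipbook of an RGBA grid texture.
# Rounded sine and cosine of time switch between R, G and B layers, each tinted by monochromatic shifts of particle color.
# These add to the desaturated alpha layer, scaled by dynamic parameters, then pass through shared emissive adjustments.
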Material /VehicleCosmetics/Boosts/Boost_PixelTech/Materials/MM_GridTech_Painted
graph {
 properties {bAutomaticallySetUsageInEditor=false, bUsedWithNiagaraRibbons=true, blendMode=BLEND_Masked, shadingModel=MSM_Unlit, usedShadingModels=Unlit}
 nodes [
  n1 = ComponentMask(None=TextureCoordinate() {vTiling=0}) {r=true}
  n2 = AppendVector(A=@n1, B=@n1)
  n3 = ScalarParameter() {defaultValue=1, group=PatternOverlay, parameterName=PatternTileU}
  n4 = ComponentMask(None=Panner(Coordinate=Multiply(A=@n2, B=@n3), Time=ScalarParameter() {group=PatternOverlay, parameterName=Pattern_PanU}) {speedX=0.5}) {r=true}
  n5 = ComponentMask(None=TextureCoordinate() {uTiling=0}) {g=true}
  n6 = AppendVector(A=@n5, B=@n5)
  n7 = ScalarParameter() {defaultValue=1, group=PatternOverlay, parameterName=PatternTileV}
  n8 = ComponentMask(None=Panner(Coordinate=Multiply(A=@n6, B=@n7), Time=ScalarParameter() {group=PatternOverlay, parameterName=Pattern_PanV}) {speedY=0.5}) {g=true}
  n9 = Constant() {r=1}
  n10 = ScalarParameter() {defaultValue=4, group=FlipBook, parameterName=Columns}
  n11 = Divide(A=@n9, B=@n10)
  n12 = ScalarParameter() {defaultValue=4, group=FlipBook, parameterName=Rows}
  n13 = Divide(A=@n9, B=@n12)
  n14 = Multiply(A=Time(), B=ScalarParameter() {defaultValue=8, group=FlipBook, parameterName=FB_Speed})
  n15 = Fmod(A=Floor(None=@n14), B=@n12)
  n16 = Floor(None=Divide(A=@n14, B=@n12))
  n17 = Add(A=Multiply(A=AppendVector(A=@n4, B=@n8), B=AppendVector(A=@n11, B=@n13)), B=AppendVector(A=Divide(A=@n15, B=@n12), B=Divide(A=@n16, B=@n10)))
  n18 = TextureSampleParameter2D(UVs=@n17) {parameterName=Texture, texture=/VehicleCosmetics/Boosts/Boost_PixelTech/Textures/T_VCVFX_Boost_PixelTech_PixelGrid_RGBA}
  n19 = TextureSampleParameter2D(UVs=@n17) {parameterName=Texture, texture=&T_VCVFX_Boost_PixelTech_PixelGrid_RGBA}
  n20 = Time()
  n21 = Ceil(None=Cosine(None=@n20))
  n22 = Clamp(None=@n21)
  n23 = Ceil(None=Sine(None=@n20))
  n24 = Clamp(None=@n23)
  n25 = LinearInterpolate(A=@n19.B, B=LinearInterpolate(A=@n19.R, B=@n19.R, Alpha=@n22), Alpha=@n24)
  MaterialExpressionVertexColor_0 = ParticleColor()
  MaterialExpressionConstantClamp_0 = Saturate(None=Multiply(A=Add(A=@n18.A, B=@n25), B=@MaterialExpressionVertexColor_0.A))
  MaterialExpressionConstantClamp_1 = Saturate(None=@n23)
  MaterialExpressionConstantClamp_2 = Saturate(None=@n21)
  n26 = Constant() {r=0.25}
  n27 = DynamicParameter() {paramNames [Brighness,R,G,B]}
  n28 = Multiply(
   A=Multiply(A=Desaturation(None=@MaterialExpressionVertexColor_0.RGB, Fraction=@n26) {luminanceFactors=<rgba 0.3 0.59 0.11 0>}, B=@n27.Brighness)
   B=@n18.A
  )
  n29 = Constant() {r=0.2}
  n30 = Multiply(
   A=@MaterialExpressionVertexColor_0.RGB
   B=SquareRoot(None=MaterialFunctionCall!(Color=@n29, Shift=@MaterialExpressionVertexColor_0.RGB) {materialFunction=/DealershipSharedLibrary/FX/Materials/MaterialFunctions/MF_MonochromaticShift}.Result)
  )
  n31 = MaterialFunctionCall(Color=Constant() {r=0.8}, Shift=@MaterialExpressionVertexColor_0.RGB) {materialFunction=&MF_MonochromaticShift}
  n32 = Multiply(A=Multiply(A=@n31.Result, B=SquareRoot(None=@n31.Result)), B=@n27.R)
  n33 = MaterialFunctionCall(Color=Constant() {r=0.4}, Shift=@MaterialExpressionVertexColor_0.RGB) {materialFunction=&MF_MonochromaticShift}
  n34 = Multiply(A=Multiply(A=SquareRoot(None=@n33.Result), B=@MaterialExpressionVertexColor_0.RGB), B=@n27.G)
  n35 = LinearInterpolate(
   A=Multiply(A=Multiply(A=@n30, B=@n27.B), B=@n19.B)
   B=LinearInterpolate(A=Multiply(A=@n32, B=@n19.R), B=Multiply(A=@n34, B=@n19.G), Alpha=@n22)
   Alpha=@n24
  )
  MaterialFunctionCall(Emissive=Multiply(A=Add(A=@n28, B=@n35), B=@n27.Brighness)) {materialFunction=&MF_FX_EmissiveAdjustments}
 ]
}
