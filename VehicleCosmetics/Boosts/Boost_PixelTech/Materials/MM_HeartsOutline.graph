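# Masked Niagara sprite material that draws a pixel-heart outline for the PixelTech vehicle boost effect.
# Rounded-up dynamic parameters RG and B drive lerps over the outline texture, with B switching from RGB to R and RG blending R with itself, so having no effect.
# The result times particle color and Brightness feeds MF_FX_EmissiveAdjustments.
# The same result times AlphaSharpness, saturated, is multiplied by particle alpha to form the mask.
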Material /VehicleCosmetics/Boosts/Boost_PixelTech/Materials/MM_HeartsOutline
graph {
 properties {bAutomaticallySetUsageInEditor=false, bUsedWithNiagaraSprites=true, blendMode=BLEND_Masked, usedShadingModels=Default Lit}
 nodes [
  n1 = DynamicParameter() {paramNames [Brightness, AlphaSharpness, RG, B]}
  n2 = TextureSample() {texture=/VehicleCosmetics/Boosts/Boost_PixelTech/Textures/T_VCVFX_Boost_PixelTech_PixelHeart_Outline_RGB_FB}
  MaterialExpressionConstantClamp_24 = Saturate(None=Ceil(None=@n1.RG))
  MaterialExpressionConstantClamp_23 = Saturate(None=Ceil(None=@n1.B))
  MaterialExpressionConstantClamp_22 = Saturate(None=LinearInterpolate(A=@n2.RGB, B=LinearInterpolate(A=@n2.R, B=@n2.R, Alpha=@MaterialExpressionConstantClamp_24), Alpha=@MaterialExpressionConstantClamp_23))
  MaterialExpressionVertexColor_4 = ParticleColor()
  MaterialFunctionCall(Emissive=Multiply(A=@n1.Brightness, B=Multiply(A=@MaterialExpressionConstantClamp_22, B=@MaterialExpressionVertexColor_4.RGB))) {materialFunction=&MF_FX_EmissiveAdjustments}
  MaterialExpressionConstantClamp_21 = Saturate(None=Multiply(A=@MaterialExpressionConstantClamp_22, B=@n1.AlphaSharpness))
  Multiply(A=@MaterialExpressionVertexColor_4.A, B=@MaterialExpressionConstantClamp_21)
 ]
}
