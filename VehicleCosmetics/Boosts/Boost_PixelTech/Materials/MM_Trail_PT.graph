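# Translucent Niagara ribbon material for the PixelTech vehicle boost trail, producing streaking, pixelated rays tinted by particle color.
# Two UV sets are floor-quantized to a grid, warped by layered panning noise, then scroll a rays texture with selectable channels.
# UV gradient masks fade the ribbon ends and hollow out the middle band.
# The result feeds an emissive adjustment function, an alpha-scaled saturated mask and a switchable distortion value.
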Material /VehicleCosmetics/Boosts/Boost_PixelTech/Materials/MM_Trail_PT
graph {
 properties {
  bAutomaticallySetUsageInEditor=false
  bUsedWithNiagaraRibbons=true
  blendMode=BLEND_Translucent
  refractionMethod=RM_2DOffset
  usedShadingModels=Default Lit
 }
 nodes [
  MaterialExpressionVertexColor_2 = ParticleColor()
  n1 = ComponentMask(None=TextureCoordinate()) {g=true}
  MaterialExpressionConstantClamp_13 = Saturate(None=ConstantBiasScale(None=@n1) {bias=-0.25, scale=3})
  MaterialExpressionConstantClamp_14 = Saturate(None=ConstantBiasScale(None=OneMinus(None=@n1)) {bias=-0.3, scale=3})
  MaterialExpressionConstantClamp_12 = Saturate(None=Multiply(A=Constant() {r=6}, B=Multiply(A=@MaterialExpressionConstantClamp_13, B=@MaterialExpressionConstantClamp_14)))
  MaterialExpressionConstantClamp_11 = Saturate(None=OneMinus(None=@MaterialExpressionConstantClamp_12))
  n2 = ComponentMask(None=TextureCoordinate()) {r=true}
  MaterialExpressionConstantClamp_15 = Saturate(None=ConstantBiasScale(None=@n2) {bias=-0.01, scale=15})
  MaterialExpressionConstantClamp_16 = Saturate(None=ConstantBiasScale(None=OneMinus(None=@n2)) {bias=-0.01, scale=5})
  n3 = Multiply(A=@MaterialExpressionConstantClamp_11, B=Multiply(A=@MaterialExpressionConstantClamp_15, B=@MaterialExpressionConstantClamp_16))
  n4 = Constant() {r=32}
  n5 = Floor(None=@n4)
  n6 = Floor(None=Multiply(A=TextureCoordinate(), B=@n5))
  n7 = Constant() {r=0.5}
  n8 = Constant() {r=0.05}
  n9 = Panner(Coordinate=TextureCoordinate() {vTiling=0.125}, Time=@n8) {speedY=0.5}
  n10 = TextureSample(UVs=Panner(Coordinate=TextureCoordinate() {uTiling=0.25, vTiling=0.25}) {speedX=-0.125, speedY=0.01}) {texture=/DealershipSharedLibrary/FX/Textures/Noise/T_VCVFX_Flow02_N}
  n11 = Multiply(A=Constant() {r=0.5}, B=AppendVector(A=@n10.R, B=@n10.R))
  n12 = TextureSample(UVs=Panner(Coordinate=Add(A=@n9, B=@n11)) {speedX=-1}) {texture=/DealershipSharedLibrary/FX/Textures/Noise/T_VCVFX_DistortNoise02_RGB}
  n13 = Panner(Coordinate=TextureCoordinate() {uTiling=1.25, vTiling=-0.125}, Time=@n8)
  n14 = TextureSample(UVs=Panner(Coordinate=Add(A=@n13, B=@n11)) {speedX=-1}) {texture=&T_VCVFX_DistortNoise02_RGB}
  n15 = Multiply(A=Constant() {r=0.15}, B=ComponentMask(None=Add(A=@n12.RGB, B=@n14.RGB)) {g=true, r=true})
  n16 = Add(A=Panner(Coordinate=Divide(A=@n6, B=@n5), Time=@n7) {speedY=0.45}, B=@n15)
  n17 = Multiply(A=Time(), B=ScalarParameter() {defaultValue=2, parameterName=PanSpeed})
  n18 = StaticComponentMaskParameter(None=TextureSample(UVs=Panner(Coordinate=@n16, Time=@n17) {speedX=-1}) {texture=/VehicleCosmetics/Boosts/Boost_PixelTech/Textures/T_VCVFX_Boost_PixelTech_RaysSharp_RGB}.RGB) {defaultG=true, parameterName=MaskRGB_01}
  n19 = Floor(None=@n4)
  n20 = Floor(None=Multiply(A=TextureCoordinate() {coordinateIndex=1}, B=@n19))
  n21 = Add(A=@n15, B=Panner(Coordinate=Divide(A=@n20, B=@n19), Time=@n7) {speedY=0.75})
  n22 = StaticComponentMaskParameter(None=TextureSample(UVs=Panner(Coordinate=@n21, Time=@n17) {speedX=-1}) {texture=&T_VCVFX_Boost_PixelTech_RaysSharp_RGB}.RGB) {defaultG=true, parameterName=MaskRGB_02}
  n23 = Multiply(A=@n3, B=Add(A=Multiply(A=@MaterialExpressionConstantClamp_16, B=@n18), B=Multiply(A=@MaterialExpressionConstantClamp_15, B=@n22)))
  MaterialExpressionConstantClamp_10 = Saturate(None=Multiply(A=Multiply(A=@MaterialExpressionVertexColor_2.A, B=@n23), B=@n3))
  MaterialFunctionCall(Emissive=Multiply(A=@MaterialExpressionVertexColor_2.RGB, B=@n23)) {materialFunction=&MF_FX_EmissiveAdjustments}
  StaticSwitchParameter(True=@n23, False=Constant()) {parameterName=Distortion}
 ]
}
