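# Masked unlit material for Niagara sprites that draws pixelated music notes for the PixelTech boost effect.
# A floored PixelSize dynamic parameter snaps UVs to a grid before sampling the animated music-note texture, so particles control pixel resolution.
# Texture RGB multiplied by particle color feeds the emissive adjustment function.
# Separate normal-lighting, specular and desaturated-alpha branches do not feed that emissive path.
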Material /VehicleCosmetics/Boosts/Boost_PixelTech/Materials/T_Notes_FB_MAT
graph {
 properties {
  bAutomaticallySetUsageInEditor=false
  bUsedWithNiagaraSprites=true
  blendMode=BLEND_Masked
  refractionMethod=RM_PixelNormalOffset
  shadingModel=MSM_Unlit
  usedShadingModels=Unlit
 }
 nodes [
  n1 = Floor(None=DynamicParameter() {paramNames [PixelSize,Param2,Param3,Param4]}.PixelSize)
  n2 = Multiply(A=ComponentMask(None=TextureCoordinate()) {g=true, r=true}, B=@n1)
  n3 = TextureSample(UVs=Divide(A=Floor(None=@n2), B=@n1)) {samplerType=SAMPLERTYPE_LinearColor, texture=/VehicleCosmetics/Boosts/Boost_PixelTech/Textures/T_VCVFX_Boost_PixelTech_MusicNotes_Anim_N}
  n4 = DotProduct(
   A=Transform(None=@n3.RGB)
   B=MaterialFunctionCall!(Texture Side A=Constant3Vector() {constant=<rgba 0 0 1 0>}, Texture Side B=Constant3Vector() {constant=<rgba 0 0 -1 0>}) {materialFunction=/Engine/Functions/Engine_MaterialFunctions03/Texturing/TwoSidedTexturing}.Texture Blend Output
  )
  n5 = Transform()
  n6 = CameraVectorWS()
  n7 = DotProduct(B=Normalize(VectorInput=Add(A=@n5, B=@n6)))
  n8 = Constant() {r=24}
  Add(A=Multiply(B=ConstantBiasScale(None=@n4) {bias=0.5}) {constA=1}, B=Multiply(A=Constant() {r=4}, B=Power(Base=@n7, Exp=@n8)))
  Add(B=Constant3Vector())
  MaterialExpressionConstantClamp_27 = Saturate(None=ConstantBiasScale(None=@n3.RGB) {bias=0.2, scale=2})
  n9 = Desaturation(None=@n3.RGB)
  MaterialExpressionConstantClamp_28 = Saturate(None=ConstantBiasScale(None=Multiply(A=@n9, B=@n3.A)) {bias=0.1, scale=1})
  Constant()
  Constant() {r=1}
  MaterialExpressionVertexColor_6 = ParticleColor()
  MaterialFunctionCall(Emissive=Multiply(A=@MaterialExpressionVertexColor_6.RGB, B=@n3.RGB)) {materialFunction=&MF_FX_EmissiveAdjustments}
 ]
}
